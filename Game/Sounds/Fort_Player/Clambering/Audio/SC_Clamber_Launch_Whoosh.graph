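# Sound cue for the player's clamber launch whoosh, with variants for local, enemy and other players.
# An IsLocal branch plays the local sound unattenuated at full mixer level, while others get basic foley attenuation at half level.
# An IsEnemy branch routes enemies to a third-person enemy sound class and others to player foley.
# All paths randomize the same six whoosh waves with slight pitch variation, and concurrency stops the oldest instance.
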SoundCue /Game/Sounds/Fort_Player/Clambering/Audio/SC_Clamber_Launch_Whoosh
root=Branch_2
nodes [
 {
  id=Attenuation_0
  children [Branch_1]
  properties {attenuationSettings=&Foley_Attenuation_Basic_Att}
 }
 {id=Branch_1, children [Mixer_4,Mixer_1,Mixer_1], properties {boolParameterName=IsEnemy}}
 {id=Branch_2, children [Mixer_3, Attenuation_0, Attenuation_0], properties {boolParameterName=IsLocal}}
 {id=Mixer_1, children [SoundClass_1], properties {inputVolume [0.5]}}
 {id=Mixer_3, children [SoundClass_3], properties {inputVolume [1]}}
 {id=Mixer_4, children [SoundClass_0], properties {inputVolume [0.5]}}
 {id=Modulator_1, children [Random_2], properties {pitchMax=1.1, pitchMin=0.9, volumeMax=1, volumeMin=1}}
 {id=Modulator_4, children [Random_7], properties {pitchMax=1.1, pitchMin=0.9, volumeMax=1, volumeMin=1}}
 {
  id=Random_2
  children [WavePlayer_9, WavePlayer_11, WavePlayer_12, WavePlayer_13, WavePlayer_14, WavePlayer_15]
  properties {bShouldExcludeFromBranchCulling=true, weights [1,1,1,1,1,1]}
 }
 {
  id=Random_7
  children [WavePlayer_19, WavePlayer_20, WavePlayer_21, WavePlayer_22, WavePlayer_23, WavePlayer_33]
  properties {bShouldExcludeFromBranchCulling=true, weights [1,1,1,1,1,1]}
 }
 {
  id=SoundClass_0
  children [Modulator_1]
  properties {soundClassOverride=&Fort_Player_Foley_3P_Enemy_Athena}
 }
 {id=SoundClass_1, children [Modulator_1], properties {soundClassOverride=&Fort_Player_Foley}}
 {id=SoundClass_3, children [Modulator_4], properties {soundClassOverride=&Fort_Player_Foley}}
 {id=WavePlayer_11, properties {soundWaveAssetPtr=&/Clamber_Launch_ZipWhoosh_08}}
 {id=WavePlayer_12, properties {soundWaveAssetPtr=&/Clamber_Launch_ZipWhoosh_07}}
 {id=WavePlayer_13, properties {soundWaveAssetPtr=&/Clamber_Launch_ZipWhoosh_06}}
 {id=WavePlayer_14, properties {soundWaveAssetPtr=&/Clamber_Launch_ZipWhoosh_05}}
 {id=WavePlayer_15, properties {soundWaveAssetPtr=&/Clamber_Launch_ZipWhoosh_04}}
 {id=WavePlayer_19, properties {soundWaveAssetPtr=&/Clamber_Launch_ZipWhoosh_10}}
 {id=WavePlayer_20, properties {soundWaveAssetPtr=&Clamber_Launch_ZipWhoosh_08}}
 {id=WavePlayer_21, properties {soundWaveAssetPtr=&Clamber_Launch_ZipWhoosh_07}}
 {id=WavePlayer_22, properties {soundWaveAssetPtr=&Clamber_Launch_ZipWhoosh_06}}
 {id=WavePlayer_23, properties {soundWaveAssetPtr=&Clamber_Launch_ZipWhoosh_05}}
 {id=WavePlayer_33, properties {soundWaveAssetPtr=&Clamber_Launch_ZipWhoosh_04}}
 {id=WavePlayer_9, properties {soundWaveAssetPtr=&Clamber_Launch_ZipWhoosh_10}}
]
properties {
 bOverrideConcurrency=true
 maxDistance=1536
 volumeMultiplier=2
 concurrencyOverrides {maxCount=4, resolutionRule=StopOldest, volumeScaleAttackTime=0.25}
}
